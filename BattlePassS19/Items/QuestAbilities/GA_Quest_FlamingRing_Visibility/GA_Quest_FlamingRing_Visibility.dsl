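; Gameplay ability for the Season 19 flaming ring quest that refreshes the ring props the owning player pawn overlaps.
; On activation from an event it calls the parent, then waits a short delay.
; It then takes the pawn's overlapping actors of the prop actor class, casts each to the flaming ring quest actor and calls UpdateOverlappingActors on it.
; The ability then ends.
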(event Ability|EventActivateAbilityFromEvent
  (params (EventData :type Gameplay%20Event%20Data%20Structure%20%28by%20ref%29))
  (bind _propactorclass (Variables|Default|GetPropActorClass))
  (bind _owningplayerpawn (Variables|GAVars|Data|GetOwningPlayerPawn))
  (bind _overlappingactors (Collision|GetOverlappingActors _owningplayerpawn _propactorclass :!class /Script/Engine.Actor))
  (|CallParentFunction EventData)
  (Utilities|FlowControl|Delay 0.6)
  (if (Utilities|Array|IsValidIndex _overlappingactors)
    (for _array_element _overlappingactors
      (bind _asbp_s19_quest_actor_flaming_ring (Utilities|Casting|CastToBP_S19_QuestActor_FlamingRing _array_element))
      (Class|BPAthenaPropQuestActorParent|UpdateOverlappingActors _asbp_s19_quest_actor_flaming_ring))
    (else
      _))
  (Ability|EndAbility))
